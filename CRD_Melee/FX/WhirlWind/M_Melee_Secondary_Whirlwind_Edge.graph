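# Translucent unlit material for Niagara mesh particles, used by name for the edge of the secondary melee whirlwind effect.
# Particle color passes through MF_FX_EmissiveAdjustments for emissive, while particle alpha scales a procedural opacity mask.
# The mask needs no textures: a sine of a power-skewed V gradient forms an off-center band, multiplied by a soft SphereMask on remapped UVs to fade it.
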Material /CRD_Melee/FX/WhirlWind/M_Melee_Secondary_Whirlwind_Edge
graph {
 properties {
  bUsedWithNiagaraMeshParticles=true
  blendMode=BLEND_Translucent
  pixelDepthOffsetMode=PDOM_Legacy
  shadingModel=MSM_Unlit
  usedShadingModels=Unlit
 }
 nodes [
  n1 = ParticleColor()
  MaterialFunctionCall(Emissive=@n1.RGB) {materialFunction=&MF_FX_EmissiveAdjustments}
  n2 = TextureCoordinate()
  n3 = ComponentMask(None=@n2) {g=true}
  n4 = OneMinus(None=@n3)
  n5 = Multiply(A=OneMinus(None=Power(Base=@n4) {constExponent=3})) {constB=0.5}
  n6 = ComponentMask(None=@n2) {r=true}
  n7 = AppendVector(A=Power(Base=OneMinus(None=@n6)), B=Multiply(A=@n3) {constB=0.5})
  n8 = Constant() {r=0.5}
  n9 = Constant()
  Multiply(A=@n1.A, B=Multiply(A=Sine(None=@n5), B=SphereMask(A=@n7, B=@n8, Radius=@n8, Hardness=@n9)))
 ]
}
